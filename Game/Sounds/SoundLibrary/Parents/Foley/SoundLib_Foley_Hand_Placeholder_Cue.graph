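# SoundLib_Foley_Hand_Placeholder_Cue is an empty Sound Cue, named as a placeholder for hand foley in the sound library's parent assets.
# It has no nodes and no root, so it produces no sound.
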SoundCue /Game/Sounds/SoundLibrary/Parents/Foley/SoundLib_Foley_Hand_Placeholder_Cue
root=null
nodes []
